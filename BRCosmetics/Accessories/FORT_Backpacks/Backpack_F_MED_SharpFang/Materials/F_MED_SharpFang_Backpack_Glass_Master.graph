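# Master translucent glass material for the SharpFang backpack cosmetic, using per-pixel lit translucency.
# Fresnel on the baked normal drives opacity, while mask textures can override opacity, metallic, specular and roughness.
# After an optional thin-film pass, an offset, scaled UV logo is composited into base color and opacity.
# An optional vertex-hide switch pulls vertices toward the object position before the shared accessory TechArt overrides.
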Material /BRCosmetics/Accessories/FORT_Backpacks/Backpack_F_MED_SharpFang/Materials/F_MED_SharpFang_Backpack_Glass_Master
graph {
 properties {
  allowTranslucentCustomDepthWrites=true
  bAutomaticallySetUsageInEditor=false
  bEnableResponsiveAA=true
  bScreenSpaceReflections=true
  bUseMaterialAttributes=true
  bUsedWithMeshDeformer=true
  bUsedWithMorphTargets=true
  bUsedWithSkeletalMesh=true
  blendMode=BLEND_Translucent
  ditherOpacityMask=true
  pixelDepthOffsetMode=PDOM_Legacy
  translucencyDirectionalLightingIntensity=5
  translucencyLightingMode=TLM_SurfacePerPixelLighting
  usedShadingModels=Default Lit
 }
 nodes [
  n1 = ScalarParameter() {parameterName=ThinFilmOpacity}
  n2 = TextureSampleParameter2D() {parameterName=Input S, texture=/Game/Global/Textures/Default/Blanks/T_BlackUserInterface}
  n3 = TextureSampleParameter2D() {parameterName=Input FX, texture=&T_BlackUserInterface}
  n4 = ScalarParameter() {parameterName=Add Custom Opacity, sortPriority=10}
  n5 = ScalarParameter() {parameterName=Fresnel Inner Transparency}
  n6 = ScalarParameter() {defaultValue=0.85, parameterName=Fresnel Outer Transparency}
  n7 = ScalarParameter() {defaultValue=1, parameterName=Fresnel Exponent}
  n8 = TextureSampleParameter2D() {
   group=Mask Texture
   parameterName=BakedNormal
   samplerSource=SSM_Wrap_WorldGroupSettings
   samplerType=SAMPLERTYPE_Normal
   texture=/Engine/EngineMaterials/FlatNormal
  }
  n9 = Transform(None=@n8.RGB)
  n10 = Clamp(None=Multiply(A=Fresnel(ExponentIn=@n7, Normal=@n9), B=ScalarParameter() {defaultValue=1, parameterName=Glass thickness}))
  n11 = LinearInterpolate(
   A=Add(A=Multiply(A=@n3.B, B=@n4), B=LinearInterpolate(A=@n5, B=@n6, Alpha=@n10))
   B=@n3.B
   Alpha=ScalarParameter() {parameterName=Use Custom Opacity}
  )
  n12 = MakeMaterialAttributes!(
   BaseColor=VectorParameter() {parameterName=Base Color}.RGB
   Metallic=LinearInterpolate(A=ScalarParameter() {defaultValue=0.1, parameterName=Metallic}, B=@n2.G, Alpha=ScalarParameter() {parameterName=Use Custom Metallic})
   Specular=LinearInterpolate(A=ScalarParameter() {defaultValue=1, parameterName=Specular}, B=@n2.R, Alpha=ScalarParameter() {parameterName=Use Custom Specular})
   Roughness=LinearInterpolate(A=ScalarParameter() {parameterName=Roughness}, B=@n2.B, Alpha=ScalarParameter() {parameterName=Use Custom Roughness})
   Anisotropy
   EmissiveColor=Constant()
   Opacity=@n11
   OpacityMask=@n11
   Normal=@n8.RGB
   Tangent
   WorldPositionOffset
   SubsurfaceColor
   ClearCoat
   ClearCoatRoughness
   AmbientOcclusion
   Refraction
   CustomizedUVs_0
   CustomizedUVs_1
   CustomizedUVs_2
   CustomizedUVs_3
   CustomizedUVs_4
   CustomizedUVs_5
   CustomizedUVs_6
   CustomizedUVs_7
   PixelDepthOffset
   ShadingModel
   Displacement
  )
  n13 = GetMaterialAttributes(None=StaticSwitchParameter(
    True=MaterialFunctionCall!(Alpha=@n1, Material Attributes=@n12) {materialFunction=/Game/Weapons/WeaponSkins/MaterialFunction/ThinFilmFunction}.Result
    False=@n12
   ) {parameterName=useThinFilm}) {attributeGetTypes [#1, #2, #3]}
  n14 = NamedRerouteUsage()
  n15 = SetMaterialAttributes!(
   MaterialAttributes=@n13.MaterialAttributes
   Base Color=LinearInterpolate(A=@n13.Base Color, B=NamedRerouteUsage(), Alpha=@n14)
   Opacity=Saturate(None=Add(A=@n13.Opacity, B=@n14))
   Opacity Mask=Saturate(None=Add(A=@n13.Opacity Mask, B=@n14))
  ) {attributeSetTypes [#1, #2, #3]}
  n16 = Multiply(A=Subtract(A=ObjectPositionWS(), B=WorldPosition().XYZ), B=ScalarParameter() {parameterName=HT_CrunchVerts})
  MaterialFunctionCall(In=StaticSwitchParameter(
    True=SetMaterialAttributes!(MaterialAttributes=@n15, World Position Offset=@n16) {attributeSetTypes [#4]}
    False=@n15
   ) {parameterName=bUseVertexHide}) {materialFunction=&MF_TechArtOverrides_accessories}
  n17 = MaterialFunctionCall(Float2=TextureCoordinate()) {materialFunction=&BreakOutFloat2Components}
  n18 = OneMinus(None=@n17.R)
  n19 = Subtract(
   A=MaterialFunctionCall(XY=MaterialFunctionCall(X=@n18, Y=@n17.G) {materialFunction=&MakeFloat2}.Result) {materialFunction=&Swizzle}.YX
   B=MaterialFunctionCall(X=ScalarParameter() {defaultValue=0.38, parameterName=Logo_Offset_Y}, Y=ScalarParameter() {defaultValue=0.362, parameterName=Logo_Offset_X}) {materialFunction=&MakeFloat2}.Result
  )
  n20 = MaterialFunctionCall(X=ScalarParameter() {defaultValue=9.1, parameterName=Logo_Scale_Y}, Y=ScalarParameter() {defaultValue=7.7, parameterName=Logo_Scale_X}) {materialFunction=&MakeFloat2}
  n21 = TextureSampleParameter2D(UVs=Saturate(None=Multiply(A=@n19, B=@n20.Result))) {
   parameterName=Logo_Texture
   samplerSource=SSM_Clamp_WorldGroupSettings
   texture=/BRCosmetics/Accessories/FORT_Backpacks/Backpack_F_MED_SharpFang/Textures/T_SharpFangGreen_Logos_MilkTea2
  }
  n22 = Subtract(A=Saturate(None=@n21.B)) {constB=0.5}
  n23 = Add(
   A=Multiply(A=@n21.G, B=VectorParameter() {parameterName=Logo_Color, defaultValue=<rgba 0.1406 0.1164 0.0989 0>}.RGB)
   B=Multiply(A=Ceil(None=@n22), B=ScalarParameter() {defaultValue=0.25, parameterName=White_Detail_Boost})
  )
  n24 = NamedRerouteDeclaration(None=@n21.A) {name=Logo_Alpha, nodeColor=<rgba 0.5234 1 0 1>}
  NamedRerouteDeclaration(None=Saturate(None=Multiply(A=@n23, B=@n24))) {name=Logo_Color, nodeColor=<rgba 0.7578 1 0 1>}
 ]
}
