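# Master material for the SharpFang backpack, drawing animated boba tea liquid inside the standard character material function chain.
# Panned, warped cloud and flow textures swirl foam into the tea, with a static switch placing cream at the bottom.
# Four pearl layers drift on sine waves and shift with Jump_Offset, two using BumpOffset for fake depth.
# Masks composite liquid and pearls into Base Color before character PostFX.
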Material /BRCosmetics/Accessories/FORT_Backpacks/Backpack_F_MED_SharpFang/Materials/F_MED_SharpFang_Backpack_Master
graph {
 properties {
  bAutomaticallySetUsageInEditor=false
  bUseMaterialAttributes=true
  bUsedWithMeshDeformer=true
  bUsedWithMorphTargets=true
  bUsedWithSkeletalMesh=true
  pixelDepthOffsetMode=PDOM_Legacy
  refractionMethod=RM_IndexOfRefraction
  shadingModel=MSM_FromMaterialExpression
  subsurfaceProfile=&SS_HeroSkin_02
  usedShadingModels=Default Lit | Subsurface | Clear Coat | Subsurface Profile | Cloth | Eye
 }
 nodes [
  n1 = Multiply(A=Time(), B=ScalarParameter() {defaultValue=0.12, group=SharpFang, parameterName=BobaWobbleSpeed})
  Add(A=@n1, B=Multiply(A=ScalarParameter() {defaultValue=1, group=SharpFang, parameterName=Jump_Offset}) {constB=0.5})
  n2 = MaterialFunctionCall() {materialFunction=&BaseFN_CharacterFunction_PreFX}
  n3 = GetMaterialAttributes(None=MaterialFunctionCall(
    PreskinnedPosition=@n2.Pre skinned Local Position
    Gmap=@n2.Pre skinned Local Position
    MaterialAttributes=@n2.Pre skinned Local Position
    AO=@n2.Pre skinned Local Position
   ) {materialFunction=&CharacterSkinFX}.Result) {attributeGetTypes [#1]}
  n4 = LinearInterpolate(
   A=NamedRerouteUsage()
   B=LinearInterpolate(
    A=VectorParameter() {group=SharpFang, parameterName=Boba_Edge_Color, defaultValue=<rgba 0.38 0.0901 0 0>}.RGB
    B=VectorParameter() {group=SharpFang, parameterName=Boba_Color}.RGB
    Alpha=NamedRerouteUsage()
   )
   Alpha=NamedRerouteUsage()
  )
  n5 = NamedRerouteUsage()
  n6 = MaterialFunctionCall(
   AO=@n2.AO
   Distance Blend=@n2.AO
   Fresnel=@n2.AO
   Gmap=@n2.AO
   Pre Skinned Local Position=@n2.AO
   Material=SetMaterialAttributes!(MaterialAttributes=@n3.MaterialAttributes, Base Color=LinearInterpolate(A=@n4, B=@n3.Base Color, Alpha=@n5)) {attributeSetTypes [#1]}
  ) {materialFunction=&BaseFN_CharacterFunction_PostFX}
  MaterialFunctionCall(In=@n6.Result) {materialFunction=&CINE_MF_CharPostFX}
  n7 = Multiply(
   A=TextureCoordinate()
   B=AppendVector(A=ScalarParameter() {defaultValue=6, group=SharpFang, parameterName=Middle Height}, B=Constant() {r=0.2})
  )
  n8 = AppendVector(A=Constant() {r=-0.5}, B=Constant() {r=-0.05})
  n9 = ComponentMask(None=SmoothStep(Min=ScalarParameter() {defaultValue=0.1, group=SharpFang, parameterName=Middle Edge Blur}, Value=Add(A=@n7, B=@n8)) {constMin=0.6}) {r=true}
  n10 = ScalarParameter() {defaultValue=1, group=SharpFang, parameterName=Mid Opacity}
  NamedRerouteDeclaration(None=Multiply(A=Multiply(A=@n9, B=@n10), B=NamedRerouteUsage())) {name=Mid Level, nodeColor=<rgba 0.3359 1 0 1>}
  NamedRerouteDeclaration(None=@n1) {name=WobbleTime, nodeColor=<rgba 1 0.375 0 1>}
  n11 = TextureSample() {samplerType=SAMPLERTYPE_Masks, texture=/BRCosmetics/Accessories/FORT_Backpacks/Backpack_F_MED_SharpFang/Textures/T_Sharpfang_Backpack_FX}
  n12 = Saturate(None=Multiply(A=Add(A=@n11.R, B=@n11.R)) {constB=2})
  n13 = SphereMask(A=TextureCoordinate(), B=Constant2Vector() {g=0.75, r=0.635}) {attenuationRadius=0.07}
  NamedRerouteDeclaration(None=Subtract(A=@n12, B=@n13)) {name=bottomcapMask, nodeColor=<rgba 1 0.2109 0 1>}
  n14 = MaterialFunctionCall(Float2=TextureCoordinate()) {materialFunction=&BreakOutFloat2Components}
  n15 = TextureCoordinate()
  n16 = Add(A=ComponentMask(None=@n15) {g=true}, B=NamedRerouteUsage())
  n17 = ScalarParameter() {defaultValue=0.2, group=SharpFang, parameterName=Layer4_OffsetTime}
  n18 = ConstantBiasScale(None=Sine(None=Add(A=@n16, B=@n17)))
  n19 = Multiply(
   A=ConstantBiasScale(None=ComponentMask(None=@n15) {r=true}) {bias=0.15, scale=0.25}
   B=Multiply(A=@n18, B=ScalarParameter() {defaultValue=0.0665, group=SharpFang, parameterName=Layer4_YDriftDistance}) {constB=0.35}
  )
  n20 = Add(
   A=Multiply(A=NamedRerouteUsage(), B=ScalarParameter() {defaultValue=0.04, group=SharpFang, parameterName=Layer4_JumpOffset}) {constB=0.75}
   B=ScalarParameter() {defaultValue=-0.0075, group=SharpFang, parameterName=Layer4_BaseOffset}
  )
  n21 = ScalarParameter() {defaultValue=0.008, group=SharpFang, parameterName=Layer4_DriftDistance}
  n22 = MaterialFunctionCall(X=Subtract(A=@n14.R, B=Add(A=@n19, B=@n20)), Y=Add(A=@n14.G, B=Multiply(A=@n18, B=@n21) {constB=0.015})) {materialFunction=&MakeFloat2}
  n23 = TextureSample(UVs=LinearInterpolate(A=TextureCoordinate(), B=@n22.Result, Alpha=NamedRerouteUsage())) {
   samplerSource=SSM_Wrap_WorldGroupSettings
   samplerType=SAMPLERTYPE_Masks
   texture=/BRCosmetics/Accessories/FORT_Backpacks/Backpack_F_MED_SharpFang/Textures/T_Sharpfang_Backpack_FX_B
  }
  NamedRerouteDeclaration(None=OneMinus(None=@n23.R)) {name=SeamMask, nodeColor=<rgba 0.7109 1 0 1>}
  n24 = ScalarParameter() {defaultValue=0.7, group=SharpFang, parameterName=Foam Edge Blur}
  n25 = ScalarParameter() {defaultValue=5, group=SharpFang, parameterName=Foam Height}
  n26 = Constant() {r=0.2}
  n27 = Add(A=Multiply(A=TextureCoordinate(), B=AppendVector(A=@n25, B=@n26)), B=AppendVector(A=Constant() {r=-0.5}, B=Constant() {r=-0.05}))
  n28 = Multiply(
   A=ComponentMask(None=SmoothStep(Min=@n24, Value=@n27) {constMin=0.6}) {r=true}
   B=ScalarParameter() {defaultValue=1, group=SharpFang, parameterName=Foam Opacity}
  )
  n29 = Clamp(None=OneMinus(None=@n28)) {minDefault=0.1}
  n30 = NamedRerouteUsage()
  NamedRerouteDeclaration(None=StaticSwitchParameter(True=Subtract(A=@n29) {constB=0.1}, False=Multiply(A=@n28, B=@n30)) {group=SharpFang, parameterName=isCreamAtBottom}) {name=Foam Level, nodeColor=<rgba 0.3359 1 0 1>}
  NamedRerouteDeclaration(None=ScalarParameter() {group=SharpFang, parameterName=Jump_Offset}) {name=Jump_Offset, nodeColor=<rgba 0 1 0.7188 1>}
  NamedRerouteDeclaration(None=@n13) {name=circle, nodeColor=<rgba 1 0.4219 0 1>}
  NamedRerouteDeclaration(None=Multiply(A=ScalarParameter() {defaultValue=0.075, group=SharpFang, parameterName=Offset_Height_Ratio}, B=NamedRerouteUsage())) {name=Masked_HeightRatio, nodeColor=<rgba 0.0312 1 0 1>}
  NamedRerouteDeclaration(None=@n2.Pre skinned Local Position) {name=PreSkinnedPos, nodeColor=<rgba 1 0 0.75 1>}
  n31 = ComponentMask(None=TextureCoordinate()) {r=true}
  n32 = ScalarParameter() {defaultValue=0.425, parameterName=Boba_BottomMask_Pos}
  n33 = NamedRerouteDeclaration(None=Ceil(None=Subtract(A=@n31, B=@n32))) {name=Bottom_Mask, nodeColor=<rgba 1 0 0.375 1>}
  n34 = ScalarParameter() {defaultValue=0.2, parameterName=Boba_Bottom_Brightness}
  n35 = OneMinus(None=@n33)
  n36 = ScalarParameter() {parameterName=Boba_EdgeMask_Offset}
  n37 = Saturate(None=Divide(A=Subtract(A=@n31, B=@n36), B=ScalarParameter() {defaultValue=0.05, parameterName=Boba_EdgeMask_Width}))
  n38 = Saturate(None=Add(A=Multiply(A=@n33, B=@n34), B=Multiply(A=@n35, B=@n37)))
  n39 = MaterialFunctionCall(Float2=TextureCoordinate()) {materialFunction=&BreakOutFloat2Components}
  n40 = TextureCoordinate()
  n41 = ConstantBiasScale(None=ComponentMask(None=@n40) {r=true}) {bias=0.15, scale=0.25}
  n42 = Add(A=ComponentMask(None=@n40) {g=true}, B=NamedRerouteUsage())
  n43 = ScalarParameter() {group=SharpFang, parameterName=Layer1_OffsetTime}
  n44 = ConstantBiasScale(None=Sine(None=Add(A=@n42, B=@n43)))
  n45 = Multiply(A=@n44, B=ScalarParameter() {defaultValue=0.1, group=SharpFang, parameterName=Layer1_YDriftDistance}) {constB=0.5}
  n46 = Multiply(A=NamedRerouteUsage(), B=ScalarParameter() {defaultValue=0.064, group=SharpFang, parameterName=Layer1_JumpOffset}) {constB=0.75}
  n47 = ScalarParameter() {group=SharpFang, parameterName=Layer1_BaseOffset}
  n48 = Subtract(A=@n39.R, B=Add(A=Multiply(A=@n41, B=@n45), B=Add(A=@n46, B=@n47)))
  n49 = Add(
   A=@n39.G
   B=Multiply(A=@n44, B=ScalarParameter() {defaultValue=-0.0112, group=SharpFang, parameterName=Layer1_XDriftDistance}) {constB=0.015}
  )
  n50 = TextureSample(UVs=BumpOffset(
    Coordinate=MaterialFunctionCall(X=@n48, Y=@n49) {materialFunction=&MakeFloat2}.Result
    Height=ScalarParameter() {defaultValue=0.66, group=SharpFang, parameterName=Layer1_OffsetDepth}
    HeightRatioInput=NamedRerouteUsage()
   )) {
   samplerSource=SSM_Wrap_WorldGroupSettings
   samplerType=SAMPLERTYPE_Masks
   texture=&T_Sharpfang_Backpack_FX_B
  }
  n51 = Saturate(None=Ceil(None=Subtract(A=@n50.G) {constB=0.9}))
  n52 = MaterialFunctionCall(Float2=TextureCoordinate()) {materialFunction=&BreakOutFloat2Components}
  n53 = TextureCoordinate()
  n54 = ComponentMask(None=@n53) {r=true}
  n55 = Add(A=ComponentMask(None=@n53) {g=true}, B=NamedRerouteUsage())
  n56 = ScalarParameter() {defaultValue=0.4, group=SharpFang, parameterName=Layer2_OffsetTime}
  n57 = ConstantBiasScale(None=Sine(None=Add(A=@n55, B=@n56)))
  n58 = ScalarParameter() {defaultValue=0.075, group=SharpFang, parameterName=Layer2_YDriftDistance}
  n59 = NamedRerouteUsage()
  n60 = ScalarParameter() {defaultValue=0.056, group=SharpFang, parameterName=Layer2_JumpOffset}
  n61 = Add(
   A=Multiply(A=ConstantBiasScale(None=@n54) {bias=0.15, scale=0.25}, B=Multiply(A=@n57, B=@n58) {constB=0.35})
   B=Add(A=Multiply(A=@n59, B=@n60) {constB=0.75}, B=ScalarParameter() {defaultValue=-0.005, group=SharpFang, parameterName=Layer2_BaseOffset})
  )
  n62 = Multiply(A=@n57, B=ScalarParameter() {defaultValue=0.016, group=SharpFang, parameterName=Layer2_DriftDistance}) {constB=0.015}
  n63 = TextureSample(UVs=MaterialFunctionCall(X=Subtract(A=@n52.R, B=@n61), Y=Add(A=@n52.G, B=@n62)) {materialFunction=&MakeFloat2}.Result) {
   samplerSource=SSM_Wrap_WorldGroupSettings
   samplerType=SAMPLERTYPE_Masks
   texture=&T_Sharpfang_Backpack_FX_B
  }
  n64 = Subtract(A=Ceil(None=Subtract(A=@n63.G) {constB=0.6}), B=Ceil(None=Subtract(A=@n63.G) {constB=0.9}))
  n65 = MaterialFunctionCall(Float2=TextureCoordinate()) {materialFunction=&BreakOutFloat2Components}
  n66 = TextureCoordinate()
  n67 = ConstantBiasScale(None=ComponentMask(None=@n66) {r=true}) {bias=0.15, scale=0.25}
  n68 = Add(A=ComponentMask(None=@n66) {g=true}, B=NamedRerouteUsage())
  n69 = ScalarParameter() {defaultValue=0.7, group=SharpFang, parameterName=Layer3_OffsetTime}
  n70 = ConstantBiasScale(None=Sine(None=Add(A=@n68, B=@n69)))
  n71 = Multiply(A=@n70, B=ScalarParameter() {defaultValue=0.05, group=SharpFang, parameterName=Layer3_YDriftDistance}) {constB=0.35}
  n72 = Multiply(A=NamedRerouteUsage(), B=ScalarParameter() {defaultValue=0.052, group=SharpFang, parameterName=Layer3_JumpOffset}) {constB=0.75}
  n73 = ScalarParameter() {defaultValue=-0.005, group=SharpFang, parameterName=Layer3_BaseOffset}
  n74 = Subtract(A=@n65.R, B=Add(A=Multiply(A=@n67, B=@n71), B=Add(A=@n72, B=@n73)))
  n75 = Add(A=@n65.G, B=Multiply(A=@n70, B=ScalarParameter() {defaultValue=-0.006, group=SharpFang, parameterName=Layer3_DriftDistance}) {constB=0.015})
  n76 = TextureSample(UVs=BumpOffset(
    Coordinate=MaterialFunctionCall(X=@n74, Y=@n75) {materialFunction=&MakeFloat2}.Result
    Height=ScalarParameter() {defaultValue=0.7, group=SharpFang, parameterName=Layer3_OffsetDepth}
    HeightRatioInput=NamedRerouteUsage()
   )) {
   samplerSource=SSM_Wrap_WorldGroupSettings
   samplerType=SAMPLERTYPE_Masks
   texture=&T_Sharpfang_Backpack_FX_B
  }
  n77 = Subtract(A=@n76.G) {constB=0.2}
  n78 = Subtract(A=@n76.G) {constB=0.6}
  n79 = Multiply(A=Subtract(A=Ceil(None=@n77), B=Ceil(None=@n78)), B=@n76.B)
  n80 = TextureSample(UVs=@n22.Result) {
   samplerSource=SSM_Wrap_WorldGroupSettings
   samplerType=SAMPLERTYPE_Masks
   texture=&T_Sharpfang_Backpack_FX_B
  }
  n81 = Ceil(None=Subtract(A=@n80.G) {constB=0.1})
  n82 = Ceil(None=Subtract(A=@n80.G) {constB=0.2})
  n83 = OneMinus(None=Ceil(None=Subtract(A=@n23.G) {constB=0.1}))
  n84 = Add(A=Multiply(A=Subtract(A=@n81, B=@n82), B=@n80.B), B=Multiply(A=Saturate(None=@n83), B=@n23.B)) {constB=0}
  n85 = Add(A=Saturate(None=Multiply(A=@n51, B=@n50.B)), B=Add(A=Multiply(A=@n64, B=@n63.B), B=Add(A=@n79, B=@n84)))
  n86 = Power(Base=Subtract(A=@n85) {constB=0.2}) {constExponent=0.5}
  n87 = Saturate(None=Add(A=Multiply(A=@n85) {constB=0.5}, B=Multiply(A=@n86) {constB=1.2}))
  n88 = Add(
   A=NamedRerouteDeclaration(None=SmoothStep(Value=@n87) {constMin=0.2}) {name=Boba_Color_Mask, nodeColor=<rgba 0.3594 0 1 1>}
   B=ScalarParameter() {defaultValue=0.6, group=SharpFang, parameterName=BobaEdgeOffset}
  )
  NamedRerouteDeclaration(None=Saturate(None=Multiply(A=@n38, B=@n88))) {name=Boba_Mask, nodeColor=<rgba 1 0.9141 0 1>}
  n89 = NamedRerouteUsage()
  n90 = NamedRerouteUsage()
  n91 = VectorParameter() {group=SharpFang, parameterName=Tea Color, defaultValue=<rgba 0.3229 0.1699 0.0706 1>}
  n92 = Multiply(A=NamedRerouteUsage(), B=@n90)
  n93 = Add(
   A=Multiply(A=VectorParameter() {group=SharpFang, parameterName=Foam Color, defaultValue=<rgba 1 1 1 1>}.RGB, B=Multiply(A=@n89, B=@n90))
   B=LinearInterpolate(
    A=Multiply(A=@n91.RGB, B=@n92)
    B=VectorParameter() {group=SharpFang, parameterName=Tea Dark Color, defaultValue=<rgba 0.1979 0.0856 0.0134 1>}.RGB
   )
  )
  n94 = VectorParameter() {group=SharpFang, parameterName=Foam Color, defaultValue=<rgba 1 1 1 1>}
  n95 = NamedRerouteUsage()
  n96 = TextureCoordinate() {uTiling=2, vTiling=2}
  n97 = Add(A=Time(), B=ScalarParameter() {group=SharpFang, parameterName=Swirl Offset})
  n98 = AppendVector(A=ScalarParameter() {defaultValue=-0.02, desc=Speed, group=SharpFang, parameterName=Speed}, B=Constant())
  n99 = Multiply(
   A=TextureSample(UVs=Panner(Coordinate=@n96, Time=@n97, Speed=@n98), Tex=NamedRerouteUsage()) {samplerType=SAMPLERTYPE_LinearColor, texture=/BRCosmetics/Athena/Environments/Nexus/Textures/zeroPointClouds_M}.A
   B=ScalarParameter() {defaultValue=0.4, group=SharpFang, parameterName=Liquid Warp Amount}
  )
  n100 = Add(
   A=TextureSample(UVs=NamedRerouteUsage(), Tex=NamedRerouteUsage()) {samplerType=SAMPLERTYPE_LinearColor, texture=&zeroPointClouds_M}.B
   B=TextureSample(UVs=Add(A=@n95, B=@n99), Tex=NamedRerouteUsage()) {samplerType=SAMPLERTYPE_LinearColor, texture=&zeroPointClouds_M}.R
  )
  n101 = Constant() {r=1.5}
  n102 = Saturate(None=Power(Base=Multiply(A=@n100, B=@n101), Exp=Constant() {r=2}))
  n103 = ComponentMask(None=Multiply(A=TextureCoordinate() {uTiling=5, vTiling=5}, B=Constant() {r=0.02})) {r=true}
  n104 = TextureCoordinate() {uTiling=5, vTiling=5}
  n105 = Panner(Coordinate=AppendVector(A=Multiply(A=@n103) {constB=21.5}, B=ComponentMask(None=@n104) {g=true})) {speedX=0.03}
  n106 = LinearInterpolate(
   A=Constant() {r=0.1}
   B=Constant() {r=0.85}
   Alpha=Add(
    A=Multiply(A=@n102) {constB=0.75}
    B=TextureSample(UVs=@n105) {samplerType=SAMPLERTYPE_LinearColor, texture=/Game/Global/Textures/Default/Flowmaps/T_SwirlFlow_02}.R
   )
  ) {constA=0.5}
  n107 = ScalarParameter() {defaultValue=2, group=SharpFang, parameterName=Foam Power}
  n108 = TextureCoordinate()
  n109 = Constant2Vector() {r=2.5}
  n110 = OneMinus(None=MaterialFunctionCall(UV Channel=Multiply(A=@n108, B=@n109)) {materialFunction=&LinearGradient}.UGradient)
  n111 = ScalarParameter() {defaultValue=0.2, group=SharpFang, parameterName=Liquid Gradient Alt}
  n112 = MaterialFunctionCall(UV Channel=Multiply(A=TextureCoordinate(), B=Constant2Vector() {r=0.5})) {materialFunction=&LinearGradient}
  n113 = ScalarParameter() {defaultValue=0.85, group=SharpFang, parameterName=Foam Swirl Opacity}
  n114 = Multiply(
   A=NamedRerouteUsage()
   B=StaticSwitchParameter(True=Multiply(A=@n110, B=@n111), False=Multiply(A=@n112.UGradient, B=@n113) {constB=0.5}) {group=SharpFang, parameterName=isCreamAtBottom}
  ) {constB=0.5}
  n115 = Add(A=@n114, B=NamedRerouteUsage())
  n116 = Multiply(
   A=Multiply(A=Multiply(A=@n106, B=@n107), B=StaticSwitchParameter(True=@n115, False=@n114) {group=SharpFang, parameterName=isCreamAtBottom})
   B=NamedRerouteUsage()
  )
  n117 = NamedRerouteDeclaration(None=@n12) {name=tea mask, nodeColor=<rgba 1 0 0.2812 1>}
  NamedRerouteDeclaration(None=Multiply(A=LinearInterpolate(A=@n93, B=@n94.RGB, Alpha=@n116), B=@n117)) {name=outSwirlColor, nodeColor=<rgba 1 0.3281 0 1>}
  n118 = OneMinus(None=Subtract(A=@n117) {constB=0.5})
  NamedRerouteDeclaration(None=Saturate(None=Floor(None=@n118))) {name=liquidMask, nodeColor=<rgba 0.9453 1 0 1>}
  n119 = Constant() {r=0.03}
  n120 = Constant()
  NamedRerouteDeclaration(None=Panner(Coordinate=@n96, Time=@n97, Speed=AppendVector(A=@n119, B=@n120))) {name=Pan UVs A, nodeColor=<rgba 0 0.1562 1 1>}
  n121 = Multiply(A=@n96, B=ScalarParameter() {defaultValue=10, group=SharpFang, parameterName=Swirl Tiling}) {constB=2}
  NamedRerouteDeclaration(None=Add(A=@n99, B=Panner(Coordinate=@n121, Time=@n97, Speed=@n98))) {name=Pan UVs 3, nodeColor=<rgba 0 0.1562 1 1>}
  NamedRerouteDeclaration(None=TextureObject() {samplerType=SAMPLERTYPE_LinearColor, texture=&zeroPointClouds_M}) {name=texObj, nodeColor=<rgba 0 0.2969 1 1>}
 ]
}
